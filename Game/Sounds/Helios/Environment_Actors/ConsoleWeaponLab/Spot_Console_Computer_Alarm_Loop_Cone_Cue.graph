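# Looping ambient cue for a weapon lab computer console, layering alarm and computer sounds over a room-tone bed.
# A mixer blends a faded-in looping bed with three randomly delayed loops: one of two alarms (one volume-enveloped), a pitched-down blip and a pitch-varied blip.
# The quality node routes this mix to two levels, leaving two empty.
# Overrides add cone attenuation with listener focus, low-pass filtering, distant reverb and limited concurrency.
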SoundCue /Game/Sounds/Helios/Environment_Actors/ConsoleWeaponLab/Spot_Console_Computer_Alarm_Loop_Cone_Cue
root=QualityLevel_0
nodes [
 {id=Delay_0, children [Random_3], properties {delayMax=8, delayMin=3}}
 {id=Delay_1, children [Modulator_0], properties {delayMax=8, delayMin=3}}
 {id=Delay_2, children [Modulator_4], properties {delayMax=8, delayMin=3}}
 {
  id=Enveloper_0
  children [Mixer_0]
  properties {pitchCurve {editorCurveData {keys [{time=0}]}}, volumeCurve {editorCurveData {keys [<at 0 1>,<at 4 1>,<at 4.5 0>]}}}
 }
 {
  id=Enveloper_1
  children [Modulator_1]
  properties {pitchCurve {editorCurveData {keys [{time=0}]}}, volumeCurve {editorCurveData {keys [<at 0 0>,<at 0.5 1>]}}}
 }
 {id=Looping_0, children [Delay_0]}
 {id=Looping_1, children [Delay_1]}
 {id=Looping_2, children [Delay_2]}
 {id=Mixer_0, children [WavePlayer_19], properties {inputVolume [1.5]}}
 {id=Mixer_1, children [WavePlayer_22], properties {inputVolume [0.25]}}
 {id=Mixer_5, children [Looping_0,Looping_1,Looping_2,Enveloper_1], properties {inputVolume [0.3,0.25,0.1,5]}}
 {id=Mixer_7, children [WavePlayer_14], properties {inputVolume [0.75]}}
 {id=Modulator_0, children [WavePlayer_4], properties {pitchMax=0.8, pitchMin=0.8, volumeMax=1, volumeMin=1}}
 {id=Modulator_1, children [WavePlayer_0], properties {pitchMax=1, pitchMin=1}}
 {id=Modulator_4, children [Mixer_7], properties {pitchMax=1.3, pitchMin=0.7, volumeMax=1, volumeMin=1}}
 {id=QualityLevel_0, children [Mixer_5,Mixer_5,null,null]}
 {id=Random_3, children [Enveloper_0,Mixer_1], properties {weights [1,1]}}
 {
  id=WavePlayer_0
  properties {bLooping=true, soundWaveAssetPtr=&/Spot_HydroBottle_Lab_ComputerRoom_Loop_Mono}
 }
 {id=WavePlayer_14, properties {soundWaveAssetPtr=/Game/Sounds/Apollo/Biomes/GovernmentSecretComplex/SecretComplex_Computer_Oneshot_15}}
 {id=WavePlayer_19, properties {soundWaveAssetPtr=/Game/Sounds/Helios/POI/WeaponLab/Spots/Spot_Weapon_Lab_Alarm_Loop_03}}
 {id=WavePlayer_22, properties {soundWaveAssetPtr=/Game/Sounds/Helios/POI/WeaponLab/Spots/Spot_Weapon_Lab_Alarm_Loop_06}}
 {id=WavePlayer_4, properties {soundWaveAssetPtr=/Game/Sounds/Apollo/Biomes/GovernmentSecretComplex/SecretComplex_Computer_Oneshot_01}}
]
properties {
 bOverrideAttenuation=true
 bOverrideConcurrency=true
 maxDistance=2750
 soundClassObject=&Fort_Gadgets
 volumeMultiplier=0.2
 attenuationOverrides {
  attenuationShape=Cone
  bEnableFocusInterpolation=true
  bEnableListenerFocus=true
  bEnableSubmixSends=true
  falloffDistance=2000
  focusAzimuth=45
  focusReleaseInterpSpeed=0.1
  lPFFrequencyAtMax=750
  lPFRadiusMax=2000
  lPFRadiusMin=750
  nonFocusVolumeAttenuation=0.2
  occlusionVolumeAttenuation=0.3
  attenuationShapeExtents=<750 20 90>
  submixSendSettings [{
    bEnableHPFCutoff=false
    bEnableLPFCutoff=false
    disableManualSendClamp=false
    hPFCutoff=20
    lPFCutoff=20000
    maxSendDistance=3000
    maxSendLevel=1
    minSendDistance=1000
    minSendLevel=0.6
    sendLevel=1
    sendLevelControlMethod=Linear
    soundSubmix=&Fortnite_Reverb_Distant
    customSendLevelCurve {editorCurveData {keys [<at 0 0.6>,<at 1 1>]}}
   }]
 }
 busSends [{audioBus=null, minSendLevel=0, sendLevel=1, soundSourceBus=null}]
 concurrencyOverrides {maxCount=2}
 soundSubmixSends [{minSendLevel=1, soundSubmix=&Fortnite_Reverb_Distant}]
}
