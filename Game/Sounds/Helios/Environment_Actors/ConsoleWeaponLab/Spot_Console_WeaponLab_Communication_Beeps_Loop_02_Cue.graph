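# Looping ambient Sound Cue for the communication beeps of a weapon lab console prop, in the Fort_Ambient_Props_Athena sound class.
# A Random node picks one of three looping beep waves with equal weight, and every quality level feeds that same node.
# Inline attenuation overrides the assigned terminal hum attenuation asset, adding short-range falloff and listener focus that favors sounds being faced.
# Inline concurrency limits playback to two instances per owner.
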SoundCue /Game/Sounds/Helios/Environment_Actors/ConsoleWeaponLab/Spot_Console_WeaponLab_Communication_Beeps_Loop_02_Cue
root=QualityLevel_0
nodes [
 {id=QualityLevel_0, children [Random_0,Random_0,Random_0,Random_0]}
 {id=Random_0, children [WavePlayer_1,WavePlayer_5,WavePlayer_6], properties {weights [1,1,1]}}
 {
  id=WavePlayer_1
  properties {bLooping=true, soundWaveAssetPtr=&/Spot_Console_Communication_Beeping_Loop_07}
 }
 {
  id=WavePlayer_5
  properties {bLooping=true, soundWaveAssetPtr=&/Spot_Console_Communication_Beeping_Loop_02}
 }
 {
  id=WavePlayer_6
  properties {bLooping=true, soundWaveAssetPtr=&/Spot_Console_Communication_Beeping_Loop_03}
 }
]
properties {
 attenuationSettings=/Game/Sounds/Helios/Environment_Actors/ConsoleTerminals/Spot_Console_Terminal_Communication_Hum_Att
 bOverrideAttenuation=true
 bOverrideConcurrency=true
 maxDistance=768
 soundClassObject=&Fort_Ambient_Props_Athena
 subtitlePriority=5000
 volumeMultiplier=0.401
 attenuationOverrides {
  bEnableFocusInterpolation=true
  bEnableListenerFocus=true
  dBAttenuationAtMax=-40
  falloffDistance=512
  focusAzimuth=15
  focusVolumeAttenuation=1.25
  nonFocusAzimuth=45
  nonFocusVolumeAttenuation=0.75
  occlusionInterpolationTime=1
  occlusionLowPassFilterFrequency=10000
  occlusionTraceChannel=ECC_GameTraceChannel11
  occlusionVolumeAttenuation=0.5
  attenuationShapeExtents=<256 64.5942 90>
 }
 concurrencyOverrides {bLimitToOwner=true, maxCount=2}
}
