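# Looping ambient tech drone hum cue for the weapon lab console, in the Athena ambient props sound class.
# Every quality level feeds one Random node that picks evenly among four looping hum waves.
# The second wave is pitched down and quieter through a Modulator and Mixer, while the fourth is boosted.
# It borrows the communication terminal hum's attenuation asset and overrides concurrency with a small per-owner limit.
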SoundCue /Game/Sounds/Helios/Environment_Actors/ConsoleWeaponLab/Spot_Console_WeaponLab_Tech_Drone_Hum_Loop_Cue
root=QualityLevel_0
nodes [
 {id=Mixer_0, children [Modulator_0], properties {inputVolume [0.8]}}
 {id=Mixer_1, children [WavePlayer_4], properties {inputVolume [1.2]}}
 {id=Modulator_0, children [WavePlayer_2], properties {pitchMax=0.8, pitchMin=0.8}}
 {id=QualityLevel_0, children [Random_1,Random_1,Random_1,Random_1]}
 {id=Random_1, children [WavePlayer_0,Mixer_0,WavePlayer_3,Mixer_1], properties {weights [1,1,1,1]}}
 {
  id=WavePlayer_0
  properties {bLooping=true, soundWaveAssetPtr=&/Spot_Console_Tech_Drone_Hum_Loop_01}
 }
 {
  id=WavePlayer_2
  properties {bLooping=true, soundWaveAssetPtr=&/Spot_Console_Tech_Drone_Hum_Loop_02}
 }
 {
  id=WavePlayer_3
  properties {bLooping=true, soundWaveAssetPtr=&/Spot_Console_Tech_Drone_Hum_Loop_03}
 }
 {
  id=WavePlayer_4
  properties {bLooping=true, soundWaveAssetPtr=&/Spot_Console_Tech_Drone_Hum_Loop_04}
 }
]
properties {
 attenuationSettings=/Game/Sounds/Helios/Environment_Actors/ConsoleTerminals/Spot_Console_Terminal_Communication_Hum_Att
 bOverrideConcurrency=true
 maxDistance=1000
 soundClassObject=&Fort_Ambient_Props_Athena
 subtitlePriority=5000
 volumeMultiplier=0.501
 attenuationOverrides {
  bEnableFocusInterpolation=true
  bEnableListenerFocus=true
  bEnableOcclusion=true
  dBAttenuationAtMax=-40
  falloffDistance=512
  focusAzimuth=15
  focusVolumeAttenuation=1.5
  nonFocusAzimuth=45
  nonFocusVolumeAttenuation=0.5
  occlusionInterpolationTime=1
  occlusionLowPassFilterFrequency=10000
  occlusionTraceChannel=ECC_GameTraceChannel11
  occlusionVolumeAttenuation=0.5
  stereoSpread=512
  attenuationShapeExtents=<256 64.5942 90>
 }
 concurrencyOverrides {bLimitToOwner=true, maxCount=2}
}
