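# Sound Cue for a quiet, looping energy hum, named as a spot ambience for a HydroBottle in the ConsoleWeaponLab environment set.
# A single looping Wave Player feeds the output directly, reusing a UFO crash site tech ambience loop in the Athena ambient props sound class.
# Inline attenuation overrides the referenced trap loop settings with a logarithmic capsule falloff and listener focus.
# A concurrency override prevents new instances once the limit is reached.
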SoundCue /Game/Sounds/Helios/Environment_Actors/ConsoleWeaponLab/Spot_HydroBottle_Energy_Hum_Loop_02_Cue
root=WavePlayer_0
nodes [{
  id=WavePlayer_0
  properties {bLooping=true, soundWaveAssetPtr=/Game/Sounds/Apollo/Biomes/UFO_Crash_Sites/AmbientLoops/AMB_UFO_Crashsite_Ambience_Tech_Loop_02}
 }]
properties {
 attenuationSettings=/Game/Sounds/Attenuation/Fort_Traps_Loop_Att
 bOverrideAttenuation=true
 bOverrideConcurrency=true
 maxDistance=921
 soundClassObject=&Fort_Ambient_Props_Athena
 volumeMultiplier=0.15
 attenuationOverrides {
  attenuationShape=Capsule
  bEnableFocusInterpolation=true
  bEnableListenerFocus=true
  distanceAlgorithm=Logarithmic
  falloffDistance=500
  nonFocusVolumeAttenuation=0.4
  stereoSpread=50
  attenuationShapeExtents=<421 100>
 }
 concurrencyOverrides {resolutionRule=PreventNew}
}
